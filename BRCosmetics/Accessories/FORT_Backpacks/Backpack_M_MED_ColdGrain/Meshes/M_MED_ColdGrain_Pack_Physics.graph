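# Physics asset for the ColdGrain backpack, defining bodies and joints for its two-sided dynamic cape.
# Left and right four-bone capsule chains hang from master bones on the backpack root, with limited swing and tight twist between segments.
# Cross constraints tie the first three left and right pairs, limiting one swing axis progressively looser down the chain.
# Pelvis, spine and thigh capsules, linked by free angular constraints, approximate the wearer's body.
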PhysicsAsset /BRCosmetics/Accessories/FORT_Backpacks/Backpack_M_MED_ColdGrain/Meshes/M_MED_ColdGrain_Pack_Physics
bodies {
 backpack_root [{shapeType=Capsule, length=1, radius=1, shapeName=backpack_root_capsule}]
 dyn_cape_l_01 [{shapeType=Capsule, length=15, radius=8, shapeName=dyn_cape_l_01_capsule, center=<8.1916 0 8.3428>, rotation=<rot 90 0 0>}]
 dyn_cape_l_02 [{shapeType=Capsule, length=15, radius=8, shapeName=dyn_cape_l_02_capsule, center=<8.1916 0 8.3428>, rotation=<rot 0 0 90>}]
 dyn_cape_l_03 [{shapeType=Capsule, length=15, radius=8, shapeName=dyn_cape_l_03_capsule, center=<8.1916 0 8.3428>, rotation=<rot 90 0 0>}]
 dyn_cape_l_04 [{shapeType=Capsule, length=8.3595, radius=8, shapeName=dyn_cape_l_04_capsule, center=<0 0 9.7162>, rotation=<rot 0 0 90>}]
 dyn_cape_l_MSTR [{shapeType=Capsule, length=0.505, radius=0.505, shapeName=dyn_cape_l_MSTR_capsule, rotation=<rot 0 -90 0>}]
 dyn_cape_r_01 [{shapeType=Capsule, length=15, radius=8, shapeName=dyn_cape_r_01_capsule, center=<8.1916 0 8.3428>, rotation=<rot 90 0 0>}]
 dyn_cape_r_02 [{shapeType=Capsule, length=15, radius=8, shapeName=dyn_cape_r_02_capsule, center=<8.1916 0 8.3428>, rotation=<rot 0 0 90>}]
 dyn_cape_r_03 [{shapeType=Capsule, length=15, radius=8, shapeName=dyn_cape_r_03_capsule, center=<8.1916 0 8.3428>, rotation=<rot 90 0 0>}]
 dyn_cape_r_04 [{shapeType=Capsule, length=8.3595, radius=8, shapeName=dyn_cape_r_04_capsule, center=<0 0 9.7162>, rotation=<rot 0 0 90>}]
 dyn_cape_r_MSTR [{shapeType=Capsule, length=0.505, radius=0.505, shapeName=dyn_cape_r_MSTR_capsule, rotation=<rot 0 -90 0>}]
 pelvis [{shapeType=Capsule, length=3.9684, radius=16.0316, shapeName=pelvis_capsule, center=<-5.6293 -1.4001 0>}]
 spine_01 [{shapeType=Capsule, length=9.7651, radius=10.2349, shapeName=spine_01_capsule}]
 spine_03 [{shapeType=Capsule, length=9.7651, radius=10.2349, shapeName=spine_03_capsule}]
 thigh_l [{shapeType=Capsule, length=30.279, radius=12.1094, shapeName=thigh_l_capsule, center=<-17.3692 0 0>, rotation=<rot 90 180 0>}]
 thigh_r [{shapeType=Capsule, length=30.279, radius=12.1094, shapeName=thigh_r_capsule, center=<17.3692 0 0>, rotation=<rot 90 0 0>}]
}
constraints [
 {bone1Name=dyn_cape_l_01, bone2Name=dyn_cape_l_MSTR, swing1LimitDegrees=15, swing2LimitDegrees=15, twistLimitDegrees=5}
 {bone1Name=dyn_cape_l_02, bone2Name=dyn_cape_l_01, swing1LimitDegrees=15, swing2LimitDegrees=30, twistLimitDegrees=5}
 {bone1Name=dyn_cape_l_03, bone2Name=dyn_cape_l_02, swing1LimitDegrees=15, swing2LimitDegrees=30, twistLimitDegrees=5}
 {bone1Name=dyn_cape_l_04, bone2Name=dyn_cape_l_03, swing1LimitDegrees=15, swing2LimitDegrees=30, twistLimitDegrees=5}
 {bone1Name=dyn_cape_r_01, bone2Name=dyn_cape_r_MSTR, swing1LimitDegrees=15, swing2LimitDegrees=15, twistLimitDegrees=5}
 {bone1Name=dyn_cape_r_02, bone2Name=dyn_cape_r_01, swing1LimitDegrees=15, swing2LimitDegrees=30, twistLimitDegrees=5}
 {bone1Name=dyn_cape_r_03, bone2Name=dyn_cape_r_02, swing1LimitDegrees=15, swing2LimitDegrees=30, twistLimitDegrees=5}
 {bone1Name=dyn_cape_r_04, bone2Name=dyn_cape_r_03, swing1LimitDegrees=15, swing2LimitDegrees=30, twistLimitDegrees=5}
 {
  bone1Name=dyn_cape_r_MSTR
  bone2Name=backpack_root
  swing1LimitDegrees=45
  swing1Motion=Free
  swing2LimitDegrees=45
  swing2Motion=Free
  twistLimitDegrees=45
  twistMotion=Free
 }
 {
  bone1Name=dyn_cape_l_MSTR
  bone2Name=backpack_root
  swing1LimitDegrees=45
  swing1Motion=Free
  swing2LimitDegrees=45
  swing2Motion=Free
  twistLimitDegrees=45
  twistMotion=Free
 }
 {
  bone1Name=pelvis
  bone2Name=backpack_root
  swing1LimitDegrees=45
  swing1Motion=Free
  swing2LimitDegrees=45
  swing2Motion=Free
  twistLimitDegrees=45
  twistMotion=Free
 }
 {
  bone1Name=thigh_l
  bone2Name=pelvis
  swing1LimitDegrees=45
  swing1Motion=Free
  swing2LimitDegrees=45
  swing2Motion=Free
  twistLimitDegrees=45
  twistMotion=Free
 }
 {
  bone1Name=thigh_r
  bone2Name=pelvis
  swing1LimitDegrees=45
  swing1Motion=Free
  swing2LimitDegrees=45
  swing2Motion=Free
  twistLimitDegrees=45
  twistMotion=Free
 }
 {
  bone1Name=spine_01
  bone2Name=pelvis
  swing1LimitDegrees=45
  swing1Motion=Free
  swing2LimitDegrees=45
  swing2Motion=Free
  twistLimitDegrees=45
  twistMotion=Free
 }
 {
  bone1Name=spine_03
  bone2Name=spine_01
  swing1LimitDegrees=45
  swing1Motion=Free
  swing2LimitDegrees=45
  swing2Motion=Free
  twistLimitDegrees=45
  twistMotion=Free
 }
 {
  bone1Name=dyn_cape_r_01
  bone2Name=dyn_cape_l_01
  swing1LimitDegrees=15
  swing2LimitDegrees=45
  swing2Motion=Free
  twistLimitDegrees=45
  twistMotion=Free
 }
 {
  bone1Name=dyn_cape_l_02
  bone2Name=dyn_cape_r_02
  swing1LimitDegrees=30
  swing2LimitDegrees=45
  swing2Motion=Free
  twistLimitDegrees=45
  twistMotion=Free
 }
 {
  bone1Name=dyn_cape_l_03
  bone2Name=dyn_cape_r_03
  swing1LimitDegrees=45
  swing2LimitDegrees=45
  swing2Motion=Free
  twistLimitDegrees=45
  twistMotion=Free
 }
]
