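# Animation Blueprint for the ColdGrain cosmetic backpack, posing it from the character's body mesh and simulating secondary motion from a cape_top base bone.
# AnimGraph layers the copied parent pose at pelvis and root, offsets backpack_root (with an extra blend while skydive floating), pins root to the cached body root transform, then runs RigidBody.
# EventGraph caches that transform and computes gravity and RigidBody settings each update, using a parameter asset borrowed from PocketBobble.
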AnimBlueprint /BRCosmetics/Accessories/FORT_Backpacks/Backpack_M_MED_ColdGrain/Meshes/M_MED_ColdGrain_Pack_AnimBP
target_skeleton=&/M_MED_ColdGrain_Pack_Skeleton
graphs [
 AnimGraph {nodes [
   n1 = Layered blend per bone(BasePose=Input Pose(), BlendPoses_0=Copy Pose From Mesh(SourceMeshComponent=Get Body_Mesh()) {bUseAttachedParent=true}) {bMeshSpaceRotationBlend=true, layerSetup [{branchFilters [{blendDepth=1, boneName=pelvis}, {blendDepth=1, boneName=root}]}]}
   n2 = "Transform (Modify) Bone - Bone: backpack_root"!(
    ComponentPose=Local To Component(LocalPose=@n1)
    bAlphaBoolEnabled:bool=Get bIsSkydiveFloating()
    Translation="0,-2,0"
    Rotation="0,0,10"
    Scale="1,1,1"
   ) -> (Pose) {
    alphaInputType=Bool
    rotationMode=BMM_Additive
    rotationSpace=BCS_BoneSpace
    scaleSpace=BCS_WorldSpace
    translationMode=BMM_Additive
    translationSpace=BCS_BoneSpace
    alphaBoolBlend {blendInTime=0.1, blendOutTime=0.4}
    boneToModify {boneName=backpack_root}
    rotation {roll=10}
    translation {y=-2}
   }
   n3 = "Transform (Modify) Bone - Bone: backpack_root"!(ComponentPose=@n2, Alpha:float=1, Translation="0,1,0", Rotation="0,0,0", Scale="1,1,1") -> (Pose) {
    rotationSpace=BCS_WorldSpace
    scaleSpace=BCS_WorldSpace
    translationMode=BMM_Additive
    translationSpace=BCS_BoneSpace
    boneToModify {boneName=backpack_root}
    translation {y=1}
   }
   n4 = Get!() -> (RootTransform_Location:Vector, RootTransform_Rotation:Rotator, RootTransform_Scale:Vector)
   n5 = "Transform (Modify) Bone - Bone: root"(ComponentPose=@n3, Translation=@n4, Rotation=@n4.RootTransform_Rotation, Scale=@n4.RootTransform_Scale) {
    rotationMode=BMM_Replace
    rotationSpace=BCS_WorldSpace
    scaleMode=BMM_Replace
    scaleSpace=BCS_WorldSpace
    translationMode=BMM_Replace
    translationSpace=BCS_WorldSpace
    boneToModify {boneName=root}
   }
   n6 = RigidBody!(
    ComponentPose=@n5
    Alpha
    OverrideWorldGravity:Vector="0,0,0"
    SimSpaceSettings:Sim Space Settings Structure=(WorldAlpha=0.5,VelocityScaleZ=0.4,DampingAlpha=1,MaxLinearVelocity=5000,MaxAngularVelocity=5000,MaxLinearAcceleration=2500,MaxAngularAcceleration=500,ExternalLinearDragV=(X=1,Y=1,Z=1),ExternalLinearVelocity=(X=0,Y=0,Z=0),ExternalAngularVelocity=(X=0,Y=0,Z=0))
   ) -> (Pose) {
    bClampLinearTranslationLimitToRefPose=true
    bOverrideWorldGravity=true
    overridePhysicsAsset=&/M_MED_ColdGrain_Pack_Physics
    simulationSpace=BaseBoneSpace
    baseBoneRef {boneName=cape_top}
    simSpaceSettings {
     maxAngularAcceleration=500
     maxAngularVelocity=5000
     maxLinearAcceleration=2500
     maxLinearVelocity=5000
     velocityScaleZ=0.4
     worldAlpha=0.5
     externalLinearDragV=<1 1 1>
    }
   }
   Output Pose(Result=Component To Local(ComponentPose=@n6)) {layerGroup=None}
  ]}
 EventGraph {nodes [
   Event BlueprintUpdateAnimation()
   > n1 = "Gravity Override State Function (Parameter Asset) (Nativized)"(
    CharacterPartAnimInstance=Self-Reference()
    GameTimeInSeconds=GetGameTimeInSeconds()
    Parameters=/BRCosmetics/Characters/Player/Female/Medium/Bodies/F_MED_PocketBobble/Meshes/Parts/F_MED_PocketBobble_FaceAcc_Parameters
   )
   > Set Gravity Override(Gravity Override=@n1)
   > n2 = Set RBAN Settings Out(RBAN Settings Out=@n1.OutRigidBodyAnimNodeParameters)
   n3 = Event OnCharacterCustomizationLoaded()
   > n4 = Set Body_Mesh(Body_Mesh=Get Array Item with EFort Custom Part Type(Array=@n3.CustomMeshParts))
   Set RootTransform(
    execute=@n2.then
    RootTransform:Transform=GetBoneTransform!(self:Skinned Mesh Component Object Reference=@n4, InBoneName:Name=root, TransformSpace:ERelativeTransformSpace Enum=RTS_World) -> (ReturnValue:Transform)
   ) -> (Output_Get:Transform)
  ]}
]
variables [
 Body_Mesh {type=SkeletalMeshComponent}
 Gravity Override {type=Vector}
 RBAN Settings Out {type=FortRigidBodyAnimNodeParameters}
 RootTransform {type=Transform, default {location=<0 0 0>, rotation=<rot 0 0 0>, scale=<1 1 1>}}
]
